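; Passive quest ability for the Season 15 battle pass that flags the owning player while they have no shield, evidently for an elimination quest.
; On activation it binds a shield check to the pawn's health-changed event and also runs the check immediately.
; At exactly zero shield it applies the NoShield gameplay effect to the pawn's ability system component.
; Otherwise it removes effects granting the Athena.Quests.Elim.NoShields tag.
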(event Ability|EventActivateAbility
  (|CallParentFunction)
  (Health|BindEventtoOnPawnHealthChanged (Variables|GAVars|Data|GetOwningPlayerPawn) (delegate Custom|Check%20Health))
  (CallFunction|CheckHealth))

(event Ability|EventFailedToActivatePassiveAbility
  (params (ActorInfo :type Gameplay%20Ability%20Actor%20Info%20Structure)))

(event Custom|Check%20Health
  (if (== (FortPawn|GetShield (Variables|GAVars|Data|GetOwningPlayerPawn)) 0.0)
    (GameplayEffects|ApplyGameplayEffectToSelf :self (Ability|GetAbilitySystemComponent (Variables|GAVars|Data|GetOwningPlayerPawn)) :GameplayEffectClass "/BattlepassS15/Items/QuestAbilities/Quest_Elim_NoShield/GE_Quest_Elim_NoShield.GE_Quest_Elim_NoShield_C" :Level 0.0)
    (else
      (GameplayEffects|RemoveActiveEffectswithGrantedTags (Ability|GetAbilitySystemComponent (Variables|GAVars|Data|GetOwningPlayerPawn)) "(GameplayTags=((TagName=\"Athena.Quests.Elim.NoShields\")))"))))
